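; PCG_DynamicMeshSimplify is a sample Blueprint in PCG Geometry Script interop content that simplifies dynamic meshes.
; Its ProcessDynamicMesh function takes a dynamic mesh and applies Geometry Script's Simplify to Vertex Count to it.
; The target count comes from the VertexCount variable.
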(fn ProcessDynamicMesh
  (params (InDynMesh :type Dynamic%20Mesh%20Object%20Reference))
  (GeometryScript|Simplification|ApplySimplifytoVertexCount InDynMesh (Variables|Default|GetVertexCount))
  (return 0))
