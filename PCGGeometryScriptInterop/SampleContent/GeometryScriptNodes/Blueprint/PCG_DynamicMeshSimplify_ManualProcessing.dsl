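; A PCG Blueprint node that simplifies incoming dynamic meshes to a target vertex count, as a manually processed sample element.
; It gathers dynamic mesh data from the In pin, then copies or steals each input as writable output data.
; Geometry Script's simplify-to-vertex-count runs on that mesh using the VertexCount setting.
; Each result is added to the Out pin with the original data's tags preserved.
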(fn Execute
  (params (Input :type PCGData%20Collection%20Structure%20%28by%20ref%29))
  (bind _outtaggeddata (Data|GetTypedInputsbyPinLabel Input "In" "/Script/PCG.PCGDynamicMeshData"))
  (for _array_element _outtaggeddata
    (bind _output_get (LocalVariables|SetOutData (PCG|DynamicMesh|CopyorStealInputData _array_element)))
    (GeometryScript|Simplification|ApplySimplifytoVertexCount (Variables|/Script/PCG.PCGDynamicMeshData|GetDynamicMesh _output_get) (Variables|Default|GetVertexCount))
    (bind _result (Utilities|Set|ToArray (.Tags (PCG|TaggedData|BreakTaggedData _array_element))))
    (Data|AddtoCollection (LocalVariables|GetOutputCollection) (LocalVariables|GetOutData) "Out" _result))
  (return (LocalVariables|GetOutputCollection)))
